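# Translucent, two-sided unlit smoke trail material for the Deimos Apparition heavy player skin, usable on meshes and Niagara particles.
# Power-warped, tiled UVs feed noise-distorted panning smoke textures that drive a dark-to-tint emissive with optional pulse.
# Opacity combines smoke, UV edge fades and depth fade, with separate two-sided handling including view-angle fading.
# Vertex offset blends noise displacement along normals with SplineThicken via UseSplineMesh, plus a camera offset.
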Material /Game/Characters/Player/Male/Medium/Bodies/M_Med_ApparitionHeavy/FX/M_Deimos_Apparition_Skin_SmokeTrail
graph {
 properties {
  bUsedWithMeshDeformer=true
  bUsedWithNiagaraMeshParticles=true
  bUsedWithNiagaraRibbons=true
  bUsedWithNiagaraSprites=true
  bUsedWithSkeletalMesh=true
  blendMode=BLEND_Translucent
  pixelDepthOffsetMode=PDOM_Legacy
  shadingModel=MSM_Unlit
  twoSided=true
  usedShadingModels=Unlit
 }
 nodes [
  n1 = Multiply(A=NamedRerouteUsage()) {constB=0.5}
  n2 = Multiply(
   A=VertexNormalWS()
   B=TextureSample(UVs=Panner(Coordinate=@n1) {speedX=-0.3, speedY=0.1}) {samplerType=SAMPLERTYPE_LinearColor, texture=&HighContrastNoise}.RGB
  )
  n3 = Multiply(
   A=ScalarParameter() {defaultValue=1, parameterName=WPO}
   B=MaterialFunctionCall(Float2=TextureCoordinate()) {materialFunction=&BreakOutFloat2Components}.R
  )
  n4 = ScalarParameter() {defaultValue=40, parameterName=Width}
  n5 = StaticBool()
  Add(
   A=MaterialFunctionCall(Offset Amount=ScalarParameter() {defaultValue=10, parameterName=CameraOffset}, World Position=WorldPosition().XYZ) {materialFunction=/Engine/Functions/Engine_MaterialFunctions02/WorldPositionOffset/CameraOffset}.Camera Offset
   B=LinearInterpolate(
    A=Multiply(A=@n2, B=@n3)
    B=MaterialFunctionCall(WidthBase=@n4, WidthTip=@n4, Expand U or V UV Channel=@n5) {materialFunction=&SplineThicken}.WorldPositionOffset
    Alpha=ScalarParameter() {defaultValue=1, parameterName=UseSplineMesh}
   )
  )
  n6 = VectorParameter() {parameterName=Color, defaultValue=<rgba 0.4032 0.238 0.5573 1>}
  n7 = TextureSample(UVs=Panner(Coordinate=NamedRerouteUsage()) {speedX=-0.3, speedY=0.1}) {samplerType=SAMPLERTYPE_Grayscale, texture=&LowResBlurredNoise}
  n8 = Panner(Coordinate=Add(A=NamedRerouteUsage(), B=ComponentMask(None=@n7.RGB) {g=true, r=true})) {speedX=-0.5, speedY=0.2}
  n9 = LinearInterpolate(
   A=Multiply(A=@n6.RGB, B=ScalarParameter() {defaultValue=0.03, parameterName=ColorDark}) {constB=0.3}
   B=@n6.RGB
   Alpha=MaterialFunctionCall(In=TextureSample(UVs=@n8) {texture=&T_Inky_Smoke_Tile}.R, Contrast=Constant() {r=0.1}) {materialFunction=&CheapContrast}.Result
  )
  n10 = Time()
  n11 = ConstantBiasScale(None=Sine(None=Divide(A=@n10)))
  MaterialFunctionCall(Emissive=LinearInterpolate(A=@n9, B=Multiply(A=@n9, B=@n11), Alpha=ScalarParameter() {parameterName=UseEmissivePulse})) {materialFunction=&MF_FX_EmissiveAdjustments}
  n12 = OneMinus(None=MaterialFunctionCall(Float2=TextureCoordinate()) {materialFunction=&BreakOutFloat2Components}.R)
  n13 = Multiply(A=MaterialFunctionCall(Float2=TextureCoordinate()) {materialFunction=&BreakOutFloat2Components}.G) {constB=0.5}
  n14 = Power(Base=Saturate(None=Sine(None=@n13))) {constExponent=3}
  n15 = DepthFade(Opacity=@n14, FadeDistance=Constant() {r=30})
  n16 = MaterialFunctionCall(Float2=TextureCoordinate()) {materialFunction=&BreakOutFloat2Components}
  n17 = Multiply(
   A=Multiply(A=Multiply(A=@n12, B=@n12), B=ShadingPathSwitch(Default=@n15, Deferred=@n15, Forward=@n14, Mobile=@n14))
   B=Saturate(None=Multiply(A=@n16.R) {constB=20})
  )
  n18 = NamedRerouteUsage()
  n19 = TextureSample(UVs=Panner(Coordinate=NamedRerouteUsage()) {speedX=-0.3, speedY=0.1}) {samplerType=SAMPLERTYPE_Grayscale, texture=&LowResBlurredNoise}
  n20 = ComponentMask(None=@n19.RGB) {g=true, r=true}
  n21 = TextureSample(UVs=Panner(Coordinate=Add(A=@n18, B=@n20)) {speedX=-0.3, speedY=0.1}) {texture=/Game/Global/Textures/Default/Tileable_Patterns/T_Blowing_Smoke_Pattern2_Tile_DISP}
  n22 = Multiply(A=@n21.R, B=ScalarParameter() {defaultValue=7, parameterName=Opacity})
  n23 = Multiply(A=@n17, B=@n22)
  n24 = ScalarParameter() {defaultValue=0.7, parameterName=InnerSideOpacity}
  n25 = LinearInterpolate(Alpha=@n19.R) {constA=4, constB=10}
  n26 = Multiply(
   A=Abs(None=MaterialFunctionCall("FadeSensitivity (optional)"=@n25) {materialFunction=&FadeBasedOnViewAngle}.Result)
   B=LinearInterpolate(Alpha=@n21.R) {constB=30}
  )
  n27 = Multiply(A=@n17, B=LinearInterpolate(A=Saturate(None=@n26), Alpha=ScalarParameter() {defaultValue=1, parameterName=UseSplineMesh}))
  Multiply(
   A=LinearInterpolate(A=Multiply(A=@n23, B=@n24), B=Multiply(A=@n27, B=@n22), Alpha=TwoSidedSign())
   B=ScalarParameter() {defaultValue=1, parameterName=BeamOpacity}
  )
  n28 = MaterialFunctionCall(Float2=TextureCoordinate()) {materialFunction=&BreakOutFloat2Components}
  n29 = Power(Base=@n28.R, Exp=ScalarParameter() {defaultValue=0.2}) {constExponent=0.5}
  n30 = VectorParameter() {parameterName=UV_Tile, defaultValue=<rgba 1 1 0 0>}
  NamedRerouteDeclaration(None=Multiply(A=AppendVector(A=@n29, B=@n28.G), B=ComponentMask(None=@n30.RGB) {g=true, r=true})) {name=UV_Input, nodeColor=<rgba 0 1 0.3906 1>}
 ]
}
